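; Creative-mode pickup actor that limits collection by team and class through a use-restriction component.
; On begin play it attaches to the minigame and applies team and class settings, which OnRep functions reapply.
; With touch pickup enabled, a non-pawn overlap cast to a vehicle gives the item to all passengers, then destroys the actor.
; A helper binds local team and class change events to UpdateVisibilityAndCollision, which is empty here.
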(event OnComponentBeginOverlap(PickupSphere)
  (params (OverlappedComponent :type Primitive%20Component%20Object%20Reference) (OtherActor :type Actor%20Object%20Reference) (OtherComp :type Primitive%20Component%20Object%20Reference) (OtherBodyIndex :type Integer) (bFromSweep :type Boolean) (SweepResult :type Hit%20Result%20Structure%20%28by%20ref%29))
  (if (Variables|FortPickupCreative|GetbPickupOnTouch)
    (Utilities|Casting|CastToFortPawn OtherActor
      (:CastFailed
        (bind _asfort_athena_vehicle (Utilities|Casting|CastToFortAthenaVehicle OtherActor))
        (for _array_element (Vehicle|GetAllPassengers _asfort_athena_vehicle)
          (CallFunction|GivePickuptoPawn _array_element false))
        (Actor|DestroyActor)))))

(event EventBeginPlay
  (Class|FortMinigameLogicComponent|TryAttachtoMinigame (Variables|Default|GetFortMinigameLogic))
  (Initialization|InitializeComponent (Variables|Default|GetCreative_UseRestriction_Component) (Variables|Default|GetFortMinigameLogic) 0 0 996 -1 0 1)
  (CallFunction|SetPickupSettings (Variables|Default|GetAllowed%20Team) (Variables|Default|GetAllowed%20Class) (Variables|Default|GetAffectsTeam) (Variables|Default|GetAffectsClass) (Variables|Default|GetItemsVisibleToAllowedOnly)))

(fn ApplySubscriptionToPlayerChangesTeamOrClass
  (params (Subscribe :type Boolean))
  (bind _outputdelegate (EventDispatchers|CreateEvent :!event UpdateVisibilityAndCollision))
  (bind _fortminigamelogic (Variables|Default|GetFortMinigameLogic))
  (bind _currentminigame (Variables|/Script/FortniteGame.FortMinigameLogicComponent|GetCurrentMinigame _fortminigamelogic))
  (Networking|CanExecuteCosmeticEvents
    (:True
      (for _array_element (Fortnite|GetLocalFortPlayerControllers true)
        (bind _asfort_player_state_athena (Utilities|Casting|CastToFortPlayerStateAthena (Variables|/Script/Engine.Controller|GetPlayerState _array_element)))
        (if (Variables|GetSubscribe)
          (Player|BindEventtoOnTeamIndexChangedDelegate _asfort_player_state_athena _outputdelegate)
          (Minigame|BindEventtoOnAnyLocalPlayerChangedClass _currentminigame _outputdelegate)
          (else
            (Player|UnbindEventfromOnTeamIndexChangedDelegate _asfort_player_state_athena _outputdelegate)
            (Minigame|UnbindEventfromOnAnyLocalPlayerChangedClass _currentminigame _outputdelegate)))))))

(fn BlueprintClassAllowed
  (params (Pawn :type Fort%20Pawn%20Object%20Reference))
  (return (StateChecks|IsValidClass (Variables|Default|GetCreative_UseRestriction_Component) Pawn)))

(fn BlueprintTeamAllowed
  (params (Pawn :type Fort%20Pawn%20Object%20Reference))
  (return (StateChecks|IsValidTeam (Variables|Default|GetCreative_UseRestriction_Component) Pawn)))

(fn OnRep_AffectsClass
  (CallFunction|UpdateAllowedClassSettings))

(fn OnRep_AffectsTeam
  (CallFunction|UpdateAllowedTeamSettings))

(fn OnRep_AllowedClass
  (CallFunction|UpdateAllowedClassSettings))

(fn OnRep_AllowedTeam
  (CallFunction|UpdateAllowedTeamSettings))

(fn OnRep_ItemsVisibleToAllowedOnly)

(fn SetPickupSettings
  (params (NewAllowedTeam :type Integer) (NewAllowedClass :type Integer) (NewAffectsTeam :type Integer) (NewAffectsClass :type Integer) (NewItemsVisibleToAllowedOnly :type Integer))
  (Utilities|FlowControl|Sequence
    (:then_0
      (Variables|Default|SetAffectsTeam (Variables|GetNewAffectsTeam))
      (bind _output_get (Variables|Default|SetAllowed%20Team (Variables|GetNewAllowedTeam)))
      (TeamRestrictions|SetTeamswithAffects (Variables|Default|GetCreative_UseRestriction_Component) _output_get (not (Math|Conversions|ToBoolean(Integer) (Variables|Default|GetAffectsTeam)))))
    (:then_1
      (Variables|Default|SetAffectsClass (Variables|GetNewAffectsClass))
      (bind _output_get_2 (Variables|Default|SetAllowed%20Class (Variables|GetNewAllowedClass)))
      (ClassRestrictions|SetClassWithAffects (Variables|Default|GetCreative_UseRestriction_Component) _output_get_2 (not (Math|Conversions|ToBoolean(Integer) (Variables|Default|GetAffectsClass))) false))))

(fn UpdateAllowedClassSettings
  (ClassRestrictions|SetClassWithAffects (Variables|Default|GetCreative_UseRestriction_Component) (Variables|Default|GetAllowed%20Class) (not (Math|Conversions|ToBoolean(Integer) (Variables|Default|GetAffectsClass))) false))

(fn UpdateAllowedTeamSettings
  (TeamRestrictions|SetTeamswithAffects (Variables|Default|GetCreative_UseRestriction_Component) (Variables|Default|GetAllowed%20Team) (not (Math|Conversions|ToBoolean(Integer) (Variables|Default|GetAffectsTeam)))))

(fn UpdateVisibilityAndCollision)

(fn ConstructionScript)
